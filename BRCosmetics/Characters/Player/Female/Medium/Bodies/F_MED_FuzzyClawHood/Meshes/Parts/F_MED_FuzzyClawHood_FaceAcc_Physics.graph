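# Physics asset for the face accessory part of the female medium FuzzyClawHood cosmetic, covering dynamic front hair, bangs and hood.
# Capsules and a few spheres shape these bones plus head, neck, upper spine, clavicles and upper arms.
# Branching hair chains use swing and twist limits below Lookat bones on neck_02, with roots angularly locked and left strands cross-linked.
# Hood halves are tightly limited to neck_02, while bangs and body joints rotate freely.
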PhysicsAsset /BRCosmetics/Characters/Player/Female/Medium/Bodies/F_MED_FuzzyClawHood/Meshes/Parts/F_MED_FuzzyClawHood_FaceAcc_Physics
bodies {
 clavicle_l [{shapeType=Capsule, length=11.1344, radius=6.9395, shapeName=clavicle_l_capsule, center=<6.9251 0 0>, rotation=<rot 90 0 0>}]
 clavicle_r [{shapeType=Capsule, length=11.1344, radius=6.9395, shapeName=clavicle_l_capsule, center=<-6.9251 0 0>, rotation=<rot 90 180 0>}]
 dyn_L_hairFront_0 [
  {shapeType=Sphere, radius=3.5, shapeName=dyn_L_hairFront_0_sphere, center=<7.8644 -2.9545 2.9943>}
  {shapeType=Capsule, length=7, radius=2.0936, shapeName=dyn_L_hairFront_0_capsule, center=<4.106 0 1.0338>, rotation=<rot 90 0 0>}
 ]
 dyn_L_hairFront_1 [{
   shapeType=Capsule
   length=5.6351
   radius=2.5
   shapeName=dyn_L_hairFront_1_capsule
   center=<3.557 0.4222 1.6843>
   rotation=<rot 0.0003 0.0001 104.0736>
  }]
 dyn_L_hairFront_2 [{shapeType=Capsule, length=8.2144, radius=2.5, shapeName=dyn_L_hairFront_2_capsule, center=<3.3063 0.4799 1.0338>, rotation=<rot 0 0 90>}]
 dyn_L_hairFront_3 [
  {shapeType=Capsule, length=7.7483, radius=2.5, shapeName=dyn_L_hairFront_3_capsule, center=<5.8102 0.4402 0.1187>, rotation=<rot 78.2602 0 0>}
  {
   shapeType=Capsule
   length=7.2926
   radius=2
   shapeName=dyn_L_hairFront_3_capsule
   center=<2.5209 -3.277 -0.9492>
   rotation=<rot 26.8719 97.1741 82.9723>
  }
  {
   shapeType=Capsule
   length=4.4076
   radius=2.5924
   shapeName=dyn_L_hairFront_3_capsule
   center=<0.7442 4.0754 0.7598>
   rotation=<rot 71.618 90.538 58.6319>
  }
 ]
 dyn_L_hairFront_5 [{
   shapeType=Capsule
   length=5.2518
   radius=2.1006
   shapeName=dyn_L_hairFront_5_capsule
   center=<1.2165 0.0762 1.853>
   rotation=<rot 0 -86.4147 -80.4917>
  }]
 dyn_L_hairFront_6 [{
   shapeType=Capsule
   length=5.1023
   radius=1.9726
   shapeName=dyn_L_hairFront_6_capsule
   center=<4.6581 -0.2224 0.2123>
   rotation=<rot 0 82.4658 -88.9388>
  }]
 dyn_L_hairFront_7 [
  {shapeType=Sphere, radius=3.0722, shapeName=dyn_L_hairFront_7_sphere, center=<9.5854 1.6431 -1.3724>}
  {
   shapeType=Capsule
   length=10.7413
   radius=3.1861
   shapeName=dyn_L_hairFront_7_capsule
   center=<2.8676 -0.0338 2.5888>
   rotation=<rot -0.008 -85.695 -83.6264>
  }
 ]
 dyn_L_hairFront_Lookat [{shapeType=Capsule, length=5.4064, radius=2.0936, shapeName=dyn_L_hairFront_0_capsule, center=<3.3063 1.0193 0.097>, rotation=<rot 90 0 0>}]
 dyn_L_hood [{
   shapeType=Capsule
   length=23.1812
   radius=8.3902
   shapeName=dyn_L_hood_capsule
   center=<6.2891 2.8576 -2.3371>
   rotation=<rot -0.5283 -4.2146 137.2405>
  }]
 dyn_R_hairFront_0 [
  {shapeType=Sphere, radius=3, shapeName=dyn_R_hairFront_0_sphere, center=<6.0354 2.0607 1.9936>}
  {shapeType=Capsule, length=7, radius=2.0936, shapeName=dyn_R_hairFront_0_capsule, center=<4.106 0 1.0338>, rotation=<rot 90 0 0>}
 ]
 dyn_R_hairFront_1 [{
   shapeType=Capsule
   length=5
   radius=2.5
   shapeName=dyn_R_hairFront_1_capsule
   center=<3.3063 -2.4875 -0.1114>
   rotation=<rot 0 -0.0001 -9.4011>
  }]
 dyn_R_hairFront_2 [
  {shapeType=Capsule, length=5, radius=2.5, shapeName=dyn_R_hairFront_2_capsule, center=<3.3063 -2.1518 0>}
  {
   shapeType=Capsule
   length=7.6174
   radius=1.8064
   shapeName=dyn_R_hairFront_2_capsule
   center=<3.6594 -1.3516 -3.8778>
   rotation=<rot 78.9248 179.9993 -170.7794>
  }
 ]
 dyn_R_hairFront_3 [
  {
   shapeType=Capsule
   length=7.8178
   radius=2.5
   shapeName=dyn_R_hairFront_3_capsule
   center=<4.244 1.1773 0.4362>
   rotation=<rot 68.0794 -0.0001 -17.3234>
  }
  {
   shapeType=Capsule
   length=5
   radius=2
   shapeName=dyn_R_hairFront_3_capsule
   center=<2.9286 3.494 0.3264>
   rotation=<rot 69.9818 -9.281 -1.6338>
  }
 ]
 dyn_R_hairFront_5 [{
   shapeType=Capsule
   length=5.9504
   radius=2.32
   shapeName=dyn_R_hairFront_5_capsule
   center=<1.9759 -3.108 -0.4149>
   rotation=<rot -0.9214 -90.8264 -90.2538>
  }]
 dyn_R_hairFront_6 [{
   shapeType=Capsule
   length=10.9789
   radius=3.0071
   shapeName=dyn_R_hairFront_6_capsule
   center=<3.6873 -1.1612 -0.1138>
   rotation=<rot -0.7646 95.3618 -85.8702>
  }]
 dyn_R_hairFront_7 [{
   shapeType=Capsule
   length=3.9173
   radius=2.8666
   shapeName=dyn_R_hairFront_7_capsule
   center=<3.1158 -0.9721 -0.9521>
   rotation=<rot 0.0026 89.6952 -91.7161>
  }]
 dyn_R_hairFront_Lookat [{shapeType=Capsule, length=5.4064, radius=2.0936, shapeName=dyn_R_hairFront_0_capsule, center=<3.3063 1.1053 0.0449>, rotation=<rot 90 0 0>}]
 dyn_R_hood [{
   shapeType=Capsule
   length=22.9296
   radius=8.3902
   shapeName=dyn_L_hood_capsule
   center=<-6.2891 -2.8576 2.3371>
   rotation=<rot -0.5283 -4.2146 -42.7595>
  }]
 dyn_bangs_1 [{
   shapeType=Capsule
   length=3.6182
   radius=2.6385
   shapeName=dyn_bangs_1_capsule
   center=<1.0443 0.1713 -1.0879>
   rotation=<rot 5.4811 -90 -16.277>
  }]
 dyn_bangs_2 [{
   shapeType=Capsule
   length=3.8229
   radius=1.8261
   shapeName=dyn_bangs_2_capsule
   center=<1.446 0.2655 0.1817>
   rotation=<rot 0 56.6257 -88.1397>
  }]
 head [{shapeType=Capsule, length=5.148, radius=7.4325, shapeName=head_capsule, center=<3.3063 0 0>, rotation=<rot 90 0 0>}]
 neck_01 [
  {shapeType=Capsule, length=5.1082, radius=5.5512, shapeName=neck_01_capsule, center=<3.3063 0 0>, rotation=<rot 90 0 0>}
  {
   shapeType=Capsule
   length=35.6421
   radius=6.8922
   shapeName=neck_01_capsule
   center=<-2.7003 -0.3456 -0.2579>
   rotation=<rot 0 -0.0005 -8.9944>
  }
 ]
 neck_02 [{shapeType=Capsule, length=2.6678, radius=6.0799, shapeName=neck_02_capsule, center=<3.3063 0 0>, rotation=<rot 90 0 0>}]
 spine_03 [
  {shapeType=Capsule, length=0, radius=9.9065, shapeName=spine_03_capsule, center=<14.212 1.218 2.2213>}
  {shapeType=Capsule, length=16.3559, radius=13.6441, shapeName=spine_03_capsule, center=<0 0.8095 0>, rotation=<rot 90 0 0>}
 ]
 spine_04 [
  {shapeType=Capsule, length=3.6574, radius=11.5098, shapeName=spine_04_capsule, center=<-3.2476 3.906 0.32>}
  {shapeType=Capsule, length=0.8628, radius=4.5472, shapeName=spine_04_capsule, center=<4.2446 4.7828 5.1792>, rotation=<rot 0 -0.0001 5.0307>}
 ]
 spine_05 [
  {shapeType=Capsule, length=16.4886, radius=7.7919, shapeName=spine_05_capsule, center=<5.6593 -3.2573 0>}
  {shapeType=Capsule, length=39.174, radius=15.5592, shapeName=spine_05_capsule, center=<9.7682 -14.9757 0>}
 ]
 upperarm_l [{shapeType=Capsule, length=11.6577, radius=3.4628, shapeName=upperarm_l_capsule, center=<6.4584 -2.5244 1.8324>, rotation=<rot 90 0 0>}]
 upperarm_r [{shapeType=Capsule, length=16.0681, radius=4.7935, shapeName=upperarm_r_capsule, center=<-6.3177 5.7893 -1.8324>, rotation=<rot 90 0 7.63>}]
 upperarm_twist_01_l [{shapeType=Capsule, length=21.4616, radius=4.2661, shapeName=upperarm_twist_01_l_capsule, center=<12 0 0>, rotation=<rot 90 0 0>}]
 upperarm_twist_01_r [{shapeType=Capsule, length=15.7841, radius=9.9435, shapeName=upperarm_twist_01_r_capsule, center=<-12 6.2103 0>, rotation=<rot 90 0 0>}]
}
constraints [
 {
  bone1Name=head
  bone2Name=neck_02
  swing1LimitDegrees=45
  swing1Motion=Free
  swing2LimitDegrees=45
  swing2Motion=Free
  twistLimitDegrees=45
  twistMotion=Free
 }
 {bone1Name=dyn_L_hairFront_Lookat, bone2Name=neck_02, swing1LimitDegrees=30, swing2LimitDegrees=30, twistLimitDegrees=10}
 {
  bone1Name=dyn_L_hairFront_0
  bone2Name=dyn_L_hairFront_Lookat
  swing1LimitDegrees=60
  swing1Motion=Locked
  swing2LimitDegrees=25
  swing2Motion=Locked
  twistLimitDegrees=10
  twistMotion=Locked
 }
 {bone1Name=dyn_L_hairFront_1, bone2Name=dyn_L_hairFront_0, swing1LimitDegrees=30, swing2LimitDegrees=30, twistLimitDegrees=1}
 {bone1Name=dyn_L_hairFront_2, bone2Name=dyn_L_hairFront_1, swing1LimitDegrees=25, swing2LimitDegrees=30, twistLimitDegrees=10}
 {bone1Name=dyn_L_hairFront_3, bone2Name=dyn_L_hairFront_2, swing1LimitDegrees=25, swing2LimitDegrees=30, twistLimitDegrees=10}
 {bone1Name=dyn_L_hood, bone2Name=neck_02, swing1LimitDegrees=5, swing2LimitDegrees=5, twistLimitDegrees=5}
 {bone1Name=dyn_R_hairFront_Lookat, bone2Name=neck_02, swing1LimitDegrees=30, swing2LimitDegrees=30, twistLimitDegrees=10}
 {
  bone1Name=dyn_R_hairFront_0
  bone2Name=dyn_R_hairFront_Lookat
  swing1LimitDegrees=60
  swing1Motion=Locked
  swing2LimitDegrees=25
  swing2Motion=Locked
  twistLimitDegrees=10
  twistMotion=Locked
 }
 {bone1Name=dyn_R_hairFront_1, bone2Name=dyn_R_hairFront_0, swing1LimitDegrees=30, swing2LimitDegrees=30, twistLimitDegrees=10}
 {bone1Name=dyn_R_hairFront_2, bone2Name=dyn_R_hairFront_1, swing1LimitDegrees=30, swing2LimitDegrees=30, twistLimitDegrees=10}
 {bone1Name=dyn_R_hairFront_3, bone2Name=dyn_R_hairFront_2, swing1LimitDegrees=30, swing2LimitDegrees=30, twistLimitDegrees=10}
 {
  bone1Name=neck_02
  bone2Name=neck_01
  swing1LimitDegrees=45
  swing1Motion=Free
  swing2LimitDegrees=45
  swing2Motion=Free
  twistLimitDegrees=45
  twistMotion=Free
 }
 {
  bone1Name=dyn_bangs_1
  bone2Name=head
  swing1LimitDegrees=45
  swing1Motion=Free
  swing2LimitDegrees=45
  swing2Motion=Free
  twistLimitDegrees=45
  twistMotion=Free
 }
 {
  bone1Name=dyn_bangs_2
  bone2Name=head
  swing1LimitDegrees=45
  swing1Motion=Free
  swing2LimitDegrees=45
  swing2Motion=Free
  twistLimitDegrees=45
  twistMotion=Free
 }
 {
  bone1Name=neck_01
  bone2Name=spine_05
  swing1LimitDegrees=45
  swing1Motion=Free
  swing2LimitDegrees=45
  swing2Motion=Free
  twistLimitDegrees=45
  twistMotion=Free
 }
 {
  bone1Name=spine_04
  bone2Name=spine_03
  swing1LimitDegrees=45
  swing1Motion=Free
  swing2LimitDegrees=45
  swing2Motion=Free
  twistLimitDegrees=45
  twistMotion=Free
 }
 {
  bone1Name=spine_05
  bone2Name=spine_04
  swing1LimitDegrees=45
  swing1Motion=Free
  swing2LimitDegrees=45
  swing2Motion=Free
  twistLimitDegrees=45
  twistMotion=Free
 }
 {
  bone1Name=clavicle_l
  bone2Name=spine_05
  swing1LimitDegrees=45
  swing1Motion=Free
  swing2LimitDegrees=45
  swing2Motion=Free
  twistLimitDegrees=45
  twistMotion=Free
 }
 {
  bone1Name=upperarm_l
  bone2Name=clavicle_l
  swing1LimitDegrees=45
  swing1Motion=Free
  swing2LimitDegrees=45
  swing2Motion=Free
  twistLimitDegrees=45
  twistMotion=Free
 }
 {
  bone1Name=upperarm_r
  bone2Name=clavicle_r
  swing1LimitDegrees=45
  swing1Motion=Free
  swing2LimitDegrees=45
  swing2Motion=Free
  twistLimitDegrees=45
  twistMotion=Free
 }
 {
  bone1Name=dyn_L_hairFront_3
  bone2Name=dyn_L_hairFront_6
  swing1LimitDegrees=10
  swing2LimitDegrees=10
  twistLimitDegrees=45
  twistMotion=Free
 }
 {bone1Name=dyn_L_hairFront_5, bone2Name=dyn_L_hairFront_1, swing1LimitDegrees=25, swing2LimitDegrees=25, twistLimitDegrees=10}
 {bone1Name=dyn_L_hairFront_6, bone2Name=dyn_L_hairFront_5, swing1LimitDegrees=25, swing2LimitDegrees=25, twistLimitDegrees=10}
 {bone1Name=dyn_L_hairFront_7, bone2Name=dyn_L_hairFront_1, swing1LimitDegrees=10, swing2LimitDegrees=25, twistLimitDegrees=10}
 {bone1Name=dyn_R_hairFront_5, bone2Name=dyn_R_hairFront_1, swing1LimitDegrees=25, swing2LimitDegrees=15, twistLimitDegrees=1}
 {bone1Name=dyn_R_hairFront_7, bone2Name=dyn_R_hairFront_5, swing1LimitDegrees=25, swing2LimitDegrees=45, twistLimitDegrees=10}
 {bone1Name=dyn_R_hairFront_6, bone2Name=dyn_R_hairFront_1, swing1LimitDegrees=25, swing2LimitDegrees=10, twistLimitDegrees=10}
 {
  bone1Name=dyn_L_hood
  bone2Name=dyn_L_hairFront_0
  swing1LimitDegrees=45
  swing1Motion=Free
  swing2LimitDegrees=45
  swing2Motion=Free
  twistLimitDegrees=45
  twistMotion=Free
 }
 {bone1Name=dyn_R_hood, bone2Name=neck_02, swing1LimitDegrees=5, swing2LimitDegrees=5, twistLimitDegrees=5}
 {
  bone1Name=dyn_R_hood
  bone2Name=dyn_R_hairFront_0
  swing1LimitDegrees=45
  swing1Motion=Free
  swing2LimitDegrees=45
  swing2Motion=Free
  twistLimitDegrees=45
  twistMotion=Free
 }
 {
  bone1Name=clavicle_r
  bone2Name=spine_05
  swing1LimitDegrees=45
  swing1Motion=Free
  swing2LimitDegrees=45
  swing2Motion=Free
  twistLimitDegrees=45
  twistMotion=Free
 }
 {
  bone1Name=upperarm_twist_01_l
  bone2Name=upperarm_l
  swing1LimitDegrees=45
  swing1Motion=Free
  swing2LimitDegrees=45
  swing2Motion=Free
  twistLimitDegrees=45
  twistMotion=Free
 }
 {
  bone1Name=upperarm_twist_01_r
  bone2Name=upperarm_r
  swing1LimitDegrees=45
  swing1Motion=Free
  swing2LimitDegrees=45
  swing2Motion=Free
  twistLimitDegrees=45
  twistMotion=Free
 }
]
