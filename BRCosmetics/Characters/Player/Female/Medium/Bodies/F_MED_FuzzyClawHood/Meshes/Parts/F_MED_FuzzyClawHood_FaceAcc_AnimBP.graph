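# Animation Blueprint for the F_MED_FuzzyClawHood face accessory, adding secondary motion and hood deformation to the character cosmetic.
# It blends the parent mesh pose with upper arms copied from BodyMesh, then runs the shared FuzzyClaw deformation Control Rig.
# RigidBody, AnimDynamics on the bangs, and head-rotation constraints on hood bones follow.
# Each update sets gravity override and LOD flag, and maps deform_hood yaw to a corrective scale alpha.
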AnimBlueprint /BRCosmetics/Characters/Player/Female/Medium/Bodies/F_MED_FuzzyClawHood/Meshes/Parts/F_MED_FuzzyClawHood_FaceAcc_AnimBP
target_skeleton=&/F_MED_FuzzyClawHood_FaceAcc_Skeleton
graphs [
 AnimGraph {nodes [
   n1 = Get BodyMesh() -> (BodyMesh:Skeletal Mesh Component Object Reference)
   n2 = Control Rig!(
    Source=Layered blend per bone(BasePose=Copy Pose From Mesh() {bUseAttachedParent=true}, BlendPoses_0=Copy Pose From Mesh(SourceMeshComponent=@n1)) {layerSetup [{branchFilters [{blendDepth=1, boneName=upperarm_l}, {blendDepth=1, boneName=upperarm_r}]}]}
    Alpha
    Sliding:bool=Get bIsBaseballSliding()
    Is Greater LOD1:bool=Get LOD is Greater 1() -> (LOD is Greater 1:bool)
   ) -> (Pose) {
    alphaBoolBlend {blendOption=Linear}
    controlRigAssetReference {blueprintRigClass=/BRCosmetics/Characters/Player/Female/Medium/Bodies/F_MED_FuzzyClaw/Meshes/F_MED_FuzzyClaw_DeformationCtrlRig.F_MED_FuzzyClaw_DeformationCtrlRig_C}
   }
   n3 = RigidBody!(ComponentPose=Local To Component(LocalPose=@n2), Alpha, OverrideWorldGravity:Vector=Get Gravity Override(), ExternalForce) -> (Pose) {
    bClampLinearTranslationLimitToRefPose=true
    bOverrideWorldGravity=true
    overridePhysicsAsset=&/F_MED_FuzzyClawHood_FaceAcc_Physics
    simulationSpace=BaseBoneSpace
    alphaBoolBlend {blendOption=Linear}
    alphaScaleBiasClamp {interpSpeedDecreasing=5, interpSpeedIncreasing=5}
    baseBoneRef {boneName=neck_01}
    componentAppliedLinearAccClamp=<3000 3000 3000>
    componentLinearAccScale=<0.4 0.4 0.4>
    componentLinearVelScale=<0.6 0.6 0.6>
    overrideWorldGravity {z=-980}
    simSpaceSettings {
     maxAngularAcceleration=500
     maxAngularVelocity=3
     maxLinearAcceleration=400
     maxLinearVelocity=250
     velocityScaleZ=0
     worldAlpha=1
     externalLinearDragV=<0.1 0.1>
    }
   }
   n4 = Anim Dynamics - Bone: dyn_bangs_1(ComponentPose=@n3, Alpha=1, ExternalForce="0,0,0") {
    angularDampingOverride=1
    angularSpringConstant=200
    bAngularSpring=true
    bOverrideAngularDamping=true
    bUseGravityOverride=true
    gravityScale=0.35
    alphaBoolBlend {blendOption=Linear}
    boundBone {boneName=dyn_bangs_1}
    componentLinearAccScale=<1 1 1>
    componentLinearVelScale=<1 1 1>
    gravityOverride {z=-1000}
    physicsBodyDefinitions [{
      boundBone {boneName=dyn_bangs_1}
      boxExtents=<10 10 10>
      constraintSetup {angularLimitsMax=<0 0 20>, angularLimitsMin=<0 0 -20>, angularTarget=<0 0 0>}
      localJointOffset=<2 0 0>
     }]
   }
   n5 = Anim Dynamics - Bone: dyn_bangs_2(ComponentPose=@n4, Alpha=1, ExternalForce="0,0,0") {
    angularDampingOverride=1
    angularSpringConstant=200
    bAngularSpring=true
    bOverrideAngularDamping=true
    bUseGravityOverride=true
    gravityScale=0.35
    alphaBoolBlend {blendOption=Linear}
    boundBone {boneName=dyn_bangs_2}
    componentLinearAccScale=<1 1 1>
    componentLinearVelScale=<1 1 1>
    gravityOverride {z=-1000}
    physicsBodyDefinitions [{
      boundBone {boneName=dyn_bangs_2}
      boxExtents=<10 10 10>
      constraintSetup {angularLimitsMax=<0 0 0>, angularLimitsMin=<0 -10 0>, angularTarget=<0 0 0>}
      localJointOffset=<2 0 0>
     }]
   }
   n6 = "Constraint - dyn_R_hood (R)"(ComponentPose=@n5, Alpha=1, ConstraintWeights_0=1) {
    alphaBoolBlend {blendOption=Linear}
    boneToModify {boneName=dyn_R_hood}
    constraintSetup [{transformType=Rotation, targetBone {boneName=head}}]
    constraintWeights [1]
   }
   n7 = "Constraint - dyn_L_hood (R)"(ComponentPose=@n6, Alpha=1, ConstraintWeights_0=1) {
    alphaBoolBlend {blendOption=Linear}
    boneToModify {boneName=dyn_L_hood}
    constraintSetup [{transformType=Rotation, targetBone {boneName=head}}]
    constraintWeights [1]
   }
   n8 = "Transform (Modify) Bone - Bone: deform_hood"(ComponentPose=@n7, Alpha=Get Raw Output() -> (Raw Output:float), Translation="0,0,0", Rotation="0,0,0", Scale="1,0.8,0.6") {
    rotationSpace=BCS_BoneSpace
    scaleMode=BMM_Additive
    scaleSpace=BCS_BoneSpace
    translationSpace=BCS_BoneSpace
    alphaBoolBlend {blendOption=Linear}
    boneToModify {boneName=deform_hood}
    scale {y=0.8, z=0.6}
   }
   Output Pose(Result=Component To Local(ComponentPose=@n8)) {layerGroup=None}
  ]}
 EventGraph {nodes [
   Event BlueprintUpdateAnimation()
   > n1 = Gravity Override State Function DEPRECATED(CharacterPartAnimInstance=Self-Reference(), GameTimeInSeconds=GetGameTimeInSeconds())
   > Set Gravity Override(Gravity Override=@n1)
   n2 = GetPartSkeletalMeshComponent()
   > Set Raw Output(Raw Output:float=MapRangeClamped(
     Value=GetBoneTransform!(
      self:Skinned Mesh Component Object Reference=@n2
      InBoneName:Name=deform_hood
      TransformSpace:ERelativeTransformSpace Enum=RTS_ParentBoneSpace
     ) -> (
      ReturnValue_Location:Vector
      ReturnValue_Rotation_Roll:float
      ReturnValue_Rotation_Pitch:float
      ReturnValue_Rotation_Yaw:float
      ReturnValue_Scale:Vector
     ).ReturnValue_Rotation_Yaw
     InRangeA=-30
     InRangeB=-70
    )) -> (Output_Get:float)
   > Set LOD is Greater 1(LOD is Greater 1:bool="integer >= integer"(A=Get CurrentLOD(), B=2)) -> (Output_Get:bool)
   Event OnCharacterCustomizationLoaded()
   > n3 = Cast To PlayerPawn_Athena(Object=TryGetPawnOwner())
   > Set BodyMesh(BodyMesh=GetSkeletalMeshForPartType(self=@n3))
  ]}
]
variables [
 BodyMesh {type=SkeletalMeshComponent}
 Gravity Override {type=Vector}
 LOD is Greater 1 {type=boolean}
 Raw Output {type=number}
]
